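# Animation Blueprint for the Cyclone backpack cosmetic, adding procedural floating motion and a body-driven tilt.
# Each update, Noise_FNC turns game time into sine-wrapped Perlin noise, filling six X/Z vectors.
# The AnimGraph offsets the pack bone, then applies those vectors as additive translations to light and dyn bones at varying alphas.
# Root roll follows the owner's spine rotation, remapped, when custom animations or emotes are active, otherwise zero.
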AnimBlueprint /BRCosmetics/Accessories/FORT_Backpacks/Backpack_M_MED_Cyclone/Meshes/M_MED_Cyclone_AnimBP
target_skeleton=&/M_MED_Cyclone_Skeleton
graphs [
 AnimGraph {nodes [
   n1 = "Transform (Modify) Bone - Bone: pack"(ComponentPose=Local To Component(LocalPose=Input Pose()), Translation="0.000,-2,0.000", Rotation="0,0,0") {
    translationMode=BMM_Additive
    translationSpace=BCS_BoneSpace
    alphaBoolBlend {blendOption=Linear}
    boneToModify {boneName=pack}
    translation {y=-2}
   }
   n2 = "Transform (Modify) Bone - Bone: dyn_light_01"(ComponentPose=@n1, Alpha=1, Translation=Get TranNoiseA() -> (TranNoiseA:Vector), Rotation="0,0,0", Scale="1,1,1") {translationMode=BMM_Additive, translationSpace=BCS_BoneSpace, alphaBoolBlend {blendOption=Linear}, boneToModify {boneName=dyn_light_01}}
   n3 = "Transform (Modify) Bone - Bone: dyn_light_02"(ComponentPose=@n2, Alpha=1, Translation=Get TranNoiseB() -> (TranNoiseB:Vector), Rotation="0,0,0", Scale="1,1,1") {translationMode=BMM_Additive, translationSpace=BCS_BoneSpace, alphaBoolBlend {blendOption=Linear}, boneToModify {boneName=dyn_light_02}}
   n4 = "Transform (Modify) Bone - Bone: dyn_light_03"(ComponentPose=@n3, Alpha=1, Translation=Get TranNoiseC() -> (TranNoiseC:Vector), Rotation="0,0,0", Scale="1,1,1") {translationMode=BMM_Additive, translationSpace=BCS_BoneSpace, alphaBoolBlend {blendOption=Linear}, boneToModify {boneName=dyn_light_03}}
   n5 = "Transform (Modify) Bone - Bone: dyn_c_02"(ComponentPose=@n4, Alpha=1, Translation=Get TranNoiseD() -> (TranNoiseD:Vector), Rotation="0,0,0", Scale="1,1,1") {translationMode=BMM_Additive, translationSpace=BCS_BoneSpace, alphaBoolBlend {blendOption=Linear}, boneToModify {boneName=dyn_c_02}}
   n6 = Get TranNoiseE() -> (TranNoiseE:Vector)
   n7 = "Transform (Modify) Bone - Bone: dyn_a"(ComponentPose=@n5, Alpha=1, Translation=@n6, Rotation="0,0,0", Scale="1,1,1") {translationMode=BMM_Additive, translationSpace=BCS_BoneSpace, alphaBoolBlend {blendOption=Linear}, boneToModify {boneName=dyn_a}}
   n8 = "Transform (Modify) Bone - Bone: dyn_u"(ComponentPose=@n7, Alpha=1, Translation=@n6, Rotation="0,0,0", Scale="1,1,1") {translationMode=BMM_Additive, translationSpace=BCS_BoneSpace, alphaBoolBlend {blendOption=Linear}, boneToModify {boneName=dyn_u}}
   n9 = Get TranNoiseF() -> (TranNoiseF:Vector)
   n10 = "Transform (Modify) Bone - Bone: dyn_s"(ComponentPose=@n8, Alpha=1, Translation=@n9, Rotation="0,0,0", Scale="1,1,1") {translationMode=BMM_Additive, translationSpace=BCS_BoneSpace, alphaBoolBlend {blendOption=Linear}, boneToModify {boneName=dyn_s}}
   n11 = "Transform (Modify) Bone - Bone: dyn_c_01"(ComponentPose=@n10, Alpha=0.8, Translation=@n9, Rotation="0,0,0", Scale="1,1,1") {translationMode=BMM_Additive, translationSpace=BCS_BoneSpace, alphaBoolBlend {blendOption=Linear}, boneToModify {boneName=dyn_c_01}}
   n12 = "Transform (Modify) Bone - Bone: dyn_j"(ComponentPose=@n11, Alpha=0.5, Translation=@n9, Rotation="0,0,0", Scale="1,1,1") {translationMode=BMM_Additive, translationSpace=BCS_BoneSpace, alphaBoolBlend {blendOption=Linear}, boneToModify {boneName=dyn_j}}
   n13 = "Transform (Modify) Bone - Bone: dyn_p"(ComponentPose=@n12, Alpha=1, Translation=Get TranNoiseC(), Rotation="0,0,0", Scale="1,1,1") {translationMode=BMM_Additive, translationSpace=BCS_BoneSpace, alphaBoolBlend {blendOption=Linear}, boneToModify {boneName=dyn_p}}
   n14 = "Transform (Modify) Bone - Bone: dyn_k"(ComponentPose=@n13, Alpha=1, Translation=Get TranNoiseB(), Rotation="0,0,0", Scale="1,1,1") {translationMode=BMM_Additive, translationSpace=BCS_BoneSpace, alphaBoolBlend {blendOption=Linear}, boneToModify {boneName=dyn_k}}
   n15 = "Transform (Modify) Bone - Bone: dyn_t"(ComponentPose=@n14, Alpha=0.4, Translation=Get TranNoiseA(), Rotation="0,0,0", Scale="1,1,1") {translationMode=BMM_Additive, translationSpace=BCS_BoneSpace, alphaBoolBlend {blendOption=Linear}, boneToModify {boneName=dyn_t}}
   n16 = "Transform (Modify) Bone - Bone: root"(ComponentPose=@n15, Translation="0,0,0", Rotation=Get backpack_rotate()) {rotationMode=BMM_Additive, rotationSpace=BCS_BoneSpace, alphaBoolBlend {blendOption=Linear}, boneToModify {boneName=root}}
   Output Pose(Result=Component To Local(ComponentPose=@n16)) {layerGroup=None}
  ]}
 EventGraph {nodes [
   Event BlueprintUpdateAnimation()
   > n1 = Sequence!(execute) -> (then_0, then_1, then_2:Exec, then_3:Exec, then_4:Exec)
   > n2 = Sequence!(execute) -> (then_0, then_1, then_2)
   > n3 = Sequence!(execute) -> (then_0, then_1)
   > n4 = Noise_FNC!(execute, self, frequency:double=0.1, amp:double=1.2) -> (then, output:double)
   n5 = Noise_FNC(execute=@n3.then_1, frequency=0.5, amp=1.8)
   > Set TranNoiseA(TranNoiseA:Vector=MakeVector(X=@n4, Y=0, Z=@n5)) -> (Output_Get:Vector)
   n6 = Sequence(execute=@n2.then_1)
   > n7 = Noise_FNC(frequency=0.1, amp=1.2)
   n8 = Noise_FNC(execute=@n6.then_1, frequency=0.6, amp=1.5)
   > Set TranNoiseB(TranNoiseB:Vector=MakeVector(X=@n7, Y=0, Z=@n8)) -> (Output_Get:Vector)
   n9 = Sequence(execute=@n2.then_2)
   > n10 = Noise_FNC(frequency=0.1, amp=1)
   n11 = Noise_FNC(execute=@n9.then_1, frequency=0.6, amp=1.2)
   > Set TranNoiseC(TranNoiseC:Vector=MakeVector(X=@n10, Y=0, Z=@n11)) -> (Output_Get:Vector)
   n12 = Sequence(execute=@n1.then_1)
   > n13 = Noise_FNC(frequency=0.4, amp=0.8)
   n14 = Noise_FNC(execute=@n12.then_1, frequency=0.6, amp=1)
   > Set TranNoiseD(TranNoiseD:Vector=MakeVector(X=@n13, Y=0, Z=@n14)) -> (Output_Get:Vector)
   n15 = Sequence(execute=@n1.then_2)
   > n16 = Noise_FNC(frequency=0.4, amp=0.6)
   n17 = Noise_FNC(execute=@n15.then_1, frequency=0.6, amp=-0.8)
   > Set TranNoiseE(TranNoiseE:Vector=MakeVector(X=@n16, Y=0, Z=@n17)) -> (Output_Get:Vector)
   n18 = Sequence(execute=@n1.then_3)
   > n19 = Noise_FNC(frequency=0.2, amp=0.6)
   n20 = Noise_FNC(execute=@n18.then_1, frequency=0.45, amp=1.6)
   > Set TranNoiseF(TranNoiseF:Vector=MakeVector(X=@n19, Y=0, Z=@n20)) -> (Output_Get:Vector)
   n21 = GetDeltaTransformFromRefPose(self=GetOwnerSkeletalMeshComponent(), BoneName=spine_05, BaseName=spine_01)
   n22 = MapRangeClamped(Value=BreakRotator(InRot=BreakTransform(InTransform=@n21).Rotation), InRangeA=12, InRangeB=-30, OutRangeA=7, OutRangeB=10)
   n23 = OR Boolean!(A=Get bCanPlayCustomAnimations() -> (bCanPlayCustomAnimations:bool), B=Get bIsPlayingEmote(), C:bool) -> (ReturnValue)
   Set backpack_rotate(execute=@n1.then_4, backpack_rotate=MakeRotator(Roll=SelectFloat(A=@n22, B=0, bPickA=@n23)))
  ]}
 Noise_FNC {nodes [
   n1 = Noise_FNC!() -> (then, frequency:double=~, amp:double=~) {is=title K2Node_FunctionEntry}
   n2 = float * float(A=@n1, B=GetGameTimeInSeconds())
   n3 = float * float(A="Sin (Radians)"(A=PerlinNoise1D!(Value:float=@n2) -> (ReturnValue:float)), B=@n1.amp)
   > Return Node!(execute, output:double=@n3)
   PerlinNoise1D(Value=@n3)
  ]}
]
variables [
 backpack_rotate {type=Rotator}
 TranNoiseA {type=Vector}
 TranNoiseB {type=Vector}
 TranNoiseC {type=Vector}
 TranNoiseD {type=Vector}
 TranNoiseE {type=Vector}
 TranNoiseF {type=Vector}
]
